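# Looping indoor room-tone ambience cue for medium rooms at residential POIs, assigned to the indoor ambient sound class.
# A quality-level node selects quad mixes for the first two levels and stereo mixes for the last two.
# Each mix layers a residential interior loop with a medium room-tone loop from the Beach Resort folder.
# Playback disables attenuation, spatialization and reverb send, and a concurrency override prevents new instances past the limit.
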SoundCue /Game/Sounds/Helios/POI/Residential/Ambience_Residential_Medium_Room_Tone_Int_Cue
root=QualityLevel_0
nodes [
 {id=Mixer_0, children [WavePlayer_11, WavePlayer_2], properties {inputVolume [1,1]}}
 {id=Mixer_1, children [WavePlayer_3,WavePlayer_0], properties {inputVolume [1,1]}}
 {id=QualityLevel_0, children [Mixer_0,Mixer_0,Mixer_1,Mixer_1]}
 {id=WavePlayer_0, properties {bLooping=true, soundWaveAssetPtr=/Game/Sounds/Asteria/POI/Residential/Ambience_Residential_Interior_Stereo_3}}
 {id=WavePlayer_11, properties {bLooping=true, soundWaveAssetPtr=/Game/Sounds/Asteria/POI/Residential/Ambience_Residential_Interior_Quad_3}}
 {
  id=WavePlayer_2
  properties {bLooping=true, soundWaveAssetPtr=/Game/Sounds/Asteria/POI/Beach_Resort/Ambience_Residential_Medium_Room_Tone_Quad}
 }
 {
  id=WavePlayer_3
  properties {bLooping=true, soundWaveAssetPtr=/Game/Sounds/Asteria/POI/Beach_Resort/Ambience_Residential_Medium_Room_Tone_Stereo}
 }
]
properties {
 bOverrideConcurrency=true
 maxDistance=2097152
 soundClassObject=&Fort_Ambient_Indoors_Athena
 volumeMultiplier=0.601
 attenuationOverrides {bAttenuate=false, bEnableReverbSend=false, bSpatialize=false}
 concurrencyOverrides {resolutionRule=PreventNew}
}
